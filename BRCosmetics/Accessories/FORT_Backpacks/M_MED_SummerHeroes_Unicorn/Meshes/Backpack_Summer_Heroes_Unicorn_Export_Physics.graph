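# Physics asset for the Summer Heroes Unicorn backpack cosmetic mesh.
# It holds a single box body on the backpack_root bone, offset and slightly rotated to cover the mesh.
# It has no constraints, so the body has no jointed or articulated parts.
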PhysicsAsset /BRCosmetics/Accessories/FORT_Backpacks/M_MED_SummerHeroes_Unicorn/Meshes/Backpack_Summer_Heroes_Unicorn_Export_Physics
bodies {backpack_root [{shapeType=Box, extentX=46.3769, extentY=33.5069, extentZ=38.5707, center=<-0.5891 -15.2048 -0.1019>, rotation=<rot 0 2.8129 0>}]}
constraints []
